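# Parent material for the SprocketPoppy boat glider, layering a curve-animated reveal, ice and metallic disc look over the shared 2023 character PreFX/PostFX functions.
# Two radial sphere-mask wipes gate a World Position Offset that collapses unrevealed vertices toward a shrink origin, with camera push and a front-end origin.
# MaterialCurveA switches off the collapse at 1.
# Noise-offset sphere masks shape disc and frost regions for blending the metal and optional tech-art ice attributes.
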Material /BRCosmetics/Gadgets/Assets/VinderTech_GliderChute/Glider_SprocketPoppy_Boat/Materials/M_Glider_SprocketPoppy_Boat_Parent
graph {
 properties {
  bAutomaticallySetUsageInEditor=false
  bUseMaterialAttributes=true
  bUsedWithMeshDeformer=true
  bUsedWithSkeletalMesh=true
  shadingModel=MSM_FromMaterialExpression
  usedShadingModels=Default Lit | Subsurface | Clear Coat | Subsurface Profile | Cloth
 }
 nodes [
  n1 = MaterialFunctionCall() {materialFunction=&MF_CharacterParent_2023_PreFX}
  n2 = GetMaterialAttributes(None=MaterialFunctionCall(Material=@n1.Result, M=@n1.Result, Pre Skinned Local Position=@n1.Result, Distance Blend=@n1.Result, Fresnel=@n1.Result) {materialFunction=&MF_CharacterParent_2023_PostFX}.Result) {attributeGetTypes [#1, #2, #3]}
  n3 = VectorParameter() {parameterName=Phase2CenterOffset, defaultValue=<rgba -0.6 0 0 1>}
  n4 = Divide(A=LocalPosition().XY, B=ScalarParameter() {defaultValue=70, parameterName=distance})
  n5 = MaterialFunctionCall(
   UVs=Add(A=ComponentMask(None=@n3.RGB) {g=true, r=true}, B=@n4)
   Rotation Center=Constant()
   "Rotation Angle (0-1)"=ScalarParameter() {defaultValue=-0.1, parameterName=Phase2_Rotation}
  ) {materialFunction=&CustomRotator}
  n6 = NamedRerouteUsage()
  n7 = ScalarParameter() {parameterName=Phase2_Clamp_Min}
  n8 = ScalarParameter() {defaultValue=3, parameterName=Phase2_Clamp_Max}
  n9 = Constant()
  n10 = SphereMask(
   A=Add(
    A=MaterialFunctionCall(Vector or UVs=@n5.Rotated Values) {materialFunction=&VectorToRadialValue}.Vector Converted to Angle
    B=NamedRerouteUsage()
   )
   B=Constant()
   Radius=MaterialFunctionCall(
    Input=Clamp(None=@n6, Min=@n7, Max=@n8)
    Input Low=@n9
    Input High=Constant() {r=1}
    Target Low=@n9
    Target High=ScalarParameter() {defaultValue=2.4, parameterName=Phase2_EndValue}
   ) {materialFunction=&RemapValueRange}.Result
   Hardness=ScalarParameter() {defaultValue=0.5, parameterName=WipingHardness}
  ) {attenuationRadius=0.1, hardnessPercent=0}
  n11 = ComponentMask(None=VectorParameter() {parameterName=Phase1CenterOffset, defaultValue=<rgba 0.8 -1.1 0 1>}.RGB) {g=true, r=true}
  n12 = MaterialFunctionCall(Vector or UVs=MaterialFunctionCall(
    UVs=Add(A=@n11, B=@n4)
    Rotation Center=Constant()
    "Rotation Angle (0-1)"=ScalarParameter() {defaultValue=-0.1, parameterName=Phase1_Rotation}
   ) {materialFunction=&CustomRotator}.Rotated Values) {materialFunction=&VectorToRadialValue}
  n13 = NamedRerouteUsage()
  n14 = ScalarParameter() {parameterName=Phase1_Clamp_Min}
  n15 = ScalarParameter() {defaultValue=0.2, parameterName=Phase1_Clamp_Max}
  n16 = Constant() {r=1}
  n17 = SphereMask(
   A=Add(A=OneMinus(None=@n12.Vector Converted to Angle), B=NamedRerouteUsage())
   B=Constant()
   Radius=MaterialFunctionCall(
    Input=Clamp(None=@n13, Min=@n14, Max=@n15)
    Input Low=Constant()
    Input High=@n16
    Target Low=ScalarParameter() {parameterName=Phase2_StartValue}
    Target High=@n16
   ) {materialFunction=&RemapValueRange}.Result
   Hardness=ScalarParameter() {defaultValue=0.5, parameterName=WipingHardness}
  ) {attenuationRadius=0.1, hardnessPercent=0}
  n18 = Saturate(None=Add(
    A=NamedRerouteDeclaration(None=@n10) {name=WipingPhase2, nodeColor=<rgba 1 0.7734 0 1>}
    B=NamedRerouteDeclaration(None=@n17) {name=WipingPhase1, nodeColor=<rgba 1 0.8203 0 1>}
   ))
  n19 = TransformPosition(None=LinearInterpolate(
    A=VectorParameter() {parameterName=ShrinkOrigin, defaultValue=<rgba 0 0 0 1>}.RGB
    B=VectorParameter() {parameterName=ShrinkOrigin_FE, defaultValue=<rgba -30 150 5 1>}.RGB
    Alpha=ScalarParameter() {parameterName=IsFrontEnd}
   )) {transformType=TRANSFORMPOSSOURCE_World}
  n20 = Multiply(
   A=Transform(None=Constant3Vector() {constant=<rgba 0 0 1 0>}) {transformSourceType=TRANSFORMSOURCE_World, transformType=TRANSFORM_Camera}
   B=ScalarParameter() {defaultValue=5, parameterName=CameraPush}
  )
  n21 = Multiply(A=Multiply(A=NamedRerouteUsage(), B=OneMinus(None=@n18)), B=Subtract(A=Add(A=@n19, B=@n20), B=WorldPosition().XYZ))
  n22 = ScalarParameter() {parameterName=MaterialCurveA}
  n23 = Sign(None=OneMinus(None=Saturate(None=@n22)))
  n24 = SetMaterialAttributes!(
   MaterialAttributes=@n2.MaterialAttributes
   World Position Offset=Add(A=@n2.World Position Offset, B=LinearInterpolate(B=@n21, Alpha=@n23))
   Emissive Color
  ) {attributeSetTypes [#1, #2]}
  n25 = NamedRerouteUsage()
  BlendMaterialAttributes(
   A=NamedRerouteUsage()
   B=StaticSwitchParameter(
    True=MaterialFunctionCall!(In=@n24, InMask=@n25, InFXMask=@n25) {materialFunction=/Game/Global/MaterialFunctions/TechArt/MF_Ice_TechArt}.Result
    False=@n24
   ) {defaultValue=true, parameterName=UseIce}
   Alpha=NamedRerouteUsage()
  )
  Constant3Vector() {constant=<rgba 0 0 3000 0>}
  Frac(None=Multiply(A=Time()) {constB=2})
  Frac(None=Multiply(A=Time()) {constB=0.8})
  n26 = ComponentMask(None=NamedRerouteUsage()) {r=true}
  n27 = VectorParameter() {parameterName=UVCropping, defaultValue=<rgba 0.11 0 0.11 0>}
  NamedRerouteDeclaration(None=Multiply(
    A=Sign(None=@n26)
    B=MaterialFunctionCall!(Crop Right=@n27.R, Crop Bottom=@n27.R, Crop Top=@n27.R, Crop Left=@n27.R) {materialFunction=/Engine/Functions/Engine_MaterialFunctions01/Texturing/UVCropping}.Result
   )) {name=EmblemMask, nodeColor=<rgba 0 0.9062 1 1>}
  n28 = NamedRerouteUsage()
  n29 = SphereMask(
   A=Add(A=Multiply(A=@n28) {constB=-0.1}, B=TextureCoordinate())
   B=ComponentMask(None=VectorParameter() {parameterName=Param2, defaultValue=<rgba 0.68 0.274 0 0>}.RGB) {g=true, r=true}
   Radius=ScalarParameter() {defaultValue=0.195, parameterName=FrostMaskRadius}
  ) {attenuationRadius=0.22, hardnessPercent=90}
  NamedRerouteDeclaration(None=Multiply(A=NamedRerouteUsage(), B=OneMinus(None=@n29))) {name=FrostMask, nodeColor=<rgba 1 0.4922 0 1>}
  NamedRerouteDeclaration(None=ScalarParameter() {parameterName=MaterialCurveA}) {name=InputAnimation, nodeColor=<rgba 1 0 0.1875 1>}
  n30 = Multiply(A=TextureCoordinate(), B=ScalarParameter() {defaultValue=0.5, parameterName=NoiseTiling})
  NamedRerouteDeclaration(None=Multiply(
    A=TextureSample(UVs=@n30) {texture=/Game/Global/Textures/Default/Noise/T_Dudebro_Noise_01}.R
    B=ScalarParameter() {defaultValue=0.35, parameterName=NoiseStrength}
   )) {name=Noise, nodeColor=<rgba 1 0.5625 0 1>}
  NamedRerouteDeclaration(None=@n1.Tech Art Mask) {name=TechArtMask_Output, nodeColor=<rgba 0.5703 0 1 1>}
  n31 = VectorParameter() {defaultValue=<rgba 0.68 0.311 0 0>}
  NamedRerouteDeclaration(None=SphereMask(
    A=TextureCoordinate()
    B=ComponentMask(None=@n31.RGB) {g=true, r=true}
    Radius=ScalarParameter() {defaultValue=0.235, parameterName=DiscRadiusMax}
   ) {attenuationRadius=0.22, hardnessPercent=80}) {name=DiscMask, nodeColor=<rgba 1 0.4922 0 1>}
  NamedRerouteDeclaration(None=ScalarParameter() {parameterName=MaterialCurveB}) {name=InputAnimationB, nodeColor=<rgba 0.1719 0 1 1>}
  n32 = TextureCoordinate() {uTiling=3, vTiling=3}
  n33 = TextureCoordinate() {uTiling=5, vTiling=5}
  n34 = SetMaterialAttributes!(
   MaterialAttributes=@n24
   Metallic=Constant() {r=1}
   Specular=ScalarParameter() {defaultValue=0.2, parameterName=DiscSpecular}
   Roughness=LinearInterpolate(
    A=ScalarParameter() {defaultValue=0.2, parameterName=DiscRoughness_Low}
    B=ScalarParameter() {defaultValue=0.6, parameterName=DiscRoughness_High}
    Alpha=TextureSample(UVs=@n32) {samplerType=SAMPLERTYPE_Masks, texture=/Game/Environments/Asteria/Terrain/Textures/T_Asteria_Ice_Noise_M}.R
   )
   Emissive Color=Constant()
   Base Color=LinearInterpolate(
    A=@n2.Base Color
    B=VectorParameter() {parameterName=MetalColor, defaultValue=<rgba 0.8774 0.9479 0.8294 1>}.RGB
    Alpha=ScalarParameter() {defaultValue=0.1, parameterName=MetalColorBlend}
   )
   Normal=MaterialFunctionCall(
    Normal=TextureSampleParameter2D(UVs=@n33) {parameterName=DiscNormal, samplerType=SAMPLERTYPE_Normal, texture=/Game/Effects/Fort_Effects/Textures/Misc/T_NoiseNormal}.R
    Flatness=Constant() {r=0.96}
   ) {materialFunction=&FlattenNormal}.Result
  ) {attributeSetTypes [
    #4
    #5
    #6
    #2
    #3
    #7
   ]}
  NamedRerouteDeclaration(None=BlendMaterialAttributes(A=@n24, B=@n34, Alpha=NamedRerouteUsage())) {name=DiscMaterialAttributes_Out, nodeColor=<rgba 0.7109 1 0 1>}
  n35 = Multiply(A=NamedRerouteUsage()) {constB=0.1}
  n36 = TextureCoordinate()
  n37 = VectorParameter() {defaultValue=<rgba 0.68 0.311 0 0>}
  n38 = NamedRerouteUsage()
  n39 = ScalarParameter() {defaultValue=0.235, parameterName=DiscRadiusMax}
  NamedRerouteDeclaration(None=SphereMask(A=Add(A=@n35, B=@n36), B=ComponentMask(None=@n37.RGB) {g=true, r=true}, Radius=Multiply(A=@n38, B=@n39)) {attenuationRadius=0.22, hardnessPercent=85}) {name=DiscAnimation, nodeColor=<rgba 1 0.4922 0 1>}
  TextureSample(UVs=@n32) {samplerType=SAMPLERTYPE_Normal, texture=/Game/Global/Textures/Default/Tileable_Patterns/T_IceScratch1_N}
  TextureSample() {samplerType=SAMPLERTYPE_LinearColor, texture=/Game/Global/Materials/MaterialGym/Materials/SimpleIce/T_Perlin_Noise_M1}
 ]
}
